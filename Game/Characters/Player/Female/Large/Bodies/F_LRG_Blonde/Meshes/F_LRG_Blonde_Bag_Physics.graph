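# Physics asset for the bags worn by the large blonde female player character, letting the main bag, its flap, and the small and big back bags swing on the body.
# Capsule bodies cover each bag bone, its offset or base anchor, the pelvis and the left upper thigh.
# Each bag hinges to its anchor on one limited swing axis with twist locked, while the flap gets tight swing and twist limits against the main bag.
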PhysicsAsset /Game/Characters/Player/Female/Large/Bodies/F_LRG_Blonde/Meshes/F_LRG_Blonde_Bag_Physics
bodies {
 base_bigBag_back [{shapeType=Capsule, length=0.505, radius=0.505, shapeName=base_bigBag_back_capsule, rotation=<rot 0 -90 0>}]
 dyn_bigBag_back [{shapeType=Capsule, length=3.1654, radius=5.8604, shapeName=dyn_bigBag_back_capsule, center=<3.3727 -0.3947 3.9391>, rotation=<rot 90 0 0>}]
 dyn_mainBag [{shapeType=Capsule, length=9.2628, radius=5.5461, shapeName=dyn_mainBag_capsule, center=<12 0 2.1085>, rotation=<rot 90 0 0>}]
 dyn_mainBag_flap_l [{shapeType=Capsule, length=4, radius=1, shapeName=dyn_mainBag_flap_l_capsule, rotation=<rot 0 0 90>}]
 dyn_mainbag_Offset [{shapeType=Capsule, length=0.505, radius=0.505, shapeName=dyn_mainbag_Offset_capsule, rotation=<rot 0 -90 0>}]
 dyn_smallbag_back [{shapeType=Capsule, length=0.8826, radius=3.9038, shapeName=dyn_smallbag_back_capsule, center=<0.4316 0 3.0607>, rotation=<rot 90 0 0>}]
 dyn_smallbag_back_Offset [{shapeType=Capsule, length=0.505, radius=0.505, shapeName=dyn_smallbag_back_Offset_capsule, rotation=<rot 0 -90 0>}]
 pelvis [{shapeType=Capsule, length=1, radius=1, shapeName=pelvis_capsule, center=<-0.2163 -2.8096 0.5929>, rotation=<rot 15.3032 -90 -6.4124>}]
 thigh_twist_01_l [{
   shapeType=Capsule
   length=21.7744
   radius=10.2217
   shapeName=thigh_twist_01_l_capsule
   center=<4.7041 -1.1222 -2.3851>
   rotation=<rot 85 179.9996 180>
  }]
}
constraints [
 {bone1Name=dyn_mainBag_flap_l, bone2Name=dyn_mainBag, swing1LimitDegrees=3, swing2LimitDegrees=15, twistLimitDegrees=3}
 {
  bone1Name=dyn_mainBag
  bone2Name=dyn_mainbag_Offset
  swing1LimitDegrees=10
  swing1Motion=Locked
  swing2LimitDegrees=45
  twistLimitDegrees=0
  twistMotion=Locked
 }
 {
  bone1Name=dyn_smallbag_back
  bone2Name=dyn_smallbag_back_Offset
  swing1LimitDegrees=0
  swing1Motion=Locked
  swing2LimitDegrees=10
  twistLimitDegrees=0
  twistMotion=Locked
 }
 {
  bone1Name=dyn_bigBag_back
  bone2Name=base_bigBag_back
  swing1LimitDegrees=0
  swing1Motion=Locked
  swing2LimitDegrees=15
  twistLimitDegrees=0
  twistMotion=Locked
 }
]
