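# Physics asset for the front jacket of the large blonde female player body, defining capsule bodies for dynamic jacket bones and the torso.
# Pelvis and spine capsules, joined by angularly free constraints, form the body volume.
# Two-bone front and side chains hang from spine_05 and single back bones from spine_04, swinging mainly on one axis with tight or zero twist.
# Angularly free cross-constraints tie each lower front bone to its neighbouring side bone.
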PhysicsAsset /Game/Characters/Player/Female/Large/Bodies/F_LRG_Blonde/Meshes/F_LRG_Blonde_FrontJacket_Physics
bodies {
 dyn_jacket__bk_l_01 [{shapeType=Capsule, length=6, radius=4, shapeName=dyn_jacket__bk_l_01_capsule, center=<3.2324 0 1.3746>, rotation=<rot 0 0 90>}]
 dyn_jacket__bk_r_01 [{shapeType=Capsule, length=6, radius=4, shapeName=dyn_jacket__bk_r_01_capsule, center=<3.2324 0 1.3746>, rotation=<rot 0 0 90>}]
 dyn_jacket_fr_l_01 [{shapeType=Capsule, length=3.5, radius=3.5, shapeName=dyn_jacket_fr_l_01_capsule, rotation=<rot 0 0 90>}]
 dyn_jacket_fr_l_02 [{shapeType=Capsule, length=6.5, radius=4, shapeName=dyn_jacket_fr_l_02_capsule, center=<-1.7928 0 2.9657>, rotation=<rot 0 0 90>}]
 dyn_jacket_fr_r_01 [{shapeType=Capsule, length=3.5, radius=3.5, shapeName=dyn_jacket_fr_r_01_capsule, rotation=<rot 0 0 90>}]
 dyn_jacket_fr_r_02 [{shapeType=Capsule, length=6.5, radius=4, shapeName=dyn_jacket_fr_r_02_capsule, center=<-1.7928 0 2.9657>, rotation=<rot 0 0 90>}]
 dyn_jacket_l_01 [{shapeType=Capsule, length=3.5, radius=3.5, shapeName=dyn_jacket_l_01_capsule, center=<0 0 1.622>, rotation=<rot 0 0 90>}]
 dyn_jacket_l_02 [{shapeType=Capsule, length=5.5, radius=5, shapeName=dyn_jacket_l_02_capsule, center=<-2.0227 0 3.1536>, rotation=<rot 0 0 90>}]
 dyn_jacket_r_01 [{shapeType=Capsule, length=3.5, radius=3.5, shapeName=dyn_jacket_r_01_capsule, center=<0 0 1.622>, rotation=<rot 0 0 90>}]
 dyn_jacket_r_02 [{shapeType=Capsule, length=5.5, radius=5, shapeName=dyn_jacket_r_02_capsule, center=<-2.0227 0 3.1536>, rotation=<rot 0 0 90>}]
 pelvis [{shapeType=Capsule, length=16.6802, radius=12.4786}]
 spine_01 [{shapeType=Capsule, length=12.5934, radius=12.1396}]
 spine_02 [{shapeType=Capsule, length=4.7091, radius=11.9304}]
 spine_03 [{shapeType=Capsule, length=11.4767, radius=6.673, shapeName=spine_03_capsule, center=<6.0885 2.1698 -0.3821>}]
 spine_04 [
  {shapeType=Capsule, length=3.3388, radius=10.082, center=<-4.4734 0.0317 0>}
  {shapeType=Capsule, length=2.1994, radius=10.2391, center=<0.0749 5.0005 0>}
 ]
 spine_05 [{shapeType=Capsule, length=7.0549, radius=11.1339}]
}
constraints [
 {
  bone1Name=spine_01
  bone2Name=pelvis
  swing1LimitDegrees=45
  swing1Motion=Free
  swing2LimitDegrees=45
  swing2Motion=Free
  twistLimitDegrees=45
  twistMotion=Free
 }
 {
  bone1Name=spine_02
  bone2Name=spine_01
  swing1LimitDegrees=45
  swing1Motion=Free
  swing2LimitDegrees=45
  swing2Motion=Free
  twistLimitDegrees=45
  twistMotion=Free
 }
 {
  bone1Name=spine_04
  bone2Name=spine_03
  swing1LimitDegrees=45
  swing1Motion=Free
  swing2LimitDegrees=45
  swing2Motion=Free
  twistLimitDegrees=45
  twistMotion=Free
 }
 {
  bone1Name=spine_05
  bone2Name=spine_04
  swing1LimitDegrees=45
  swing1Motion=Free
  swing2LimitDegrees=45
  swing2Motion=Free
  twistLimitDegrees=45
  twistMotion=Free
 }
 {bone1Name=dyn_jacket_l_01, bone2Name=spine_05, swing1LimitDegrees=5, swing2LimitDegrees=70, twistLimitDegrees=3}
 {bone1Name=dyn_jacket_fr_l_01, bone2Name=spine_05, swing1LimitDegrees=25, swing2LimitDegrees=70, twistLimitDegrees=1}
 {bone1Name=dyn_jacket_r_01, bone2Name=spine_05, swing1LimitDegrees=5, swing2LimitDegrees=70, twistLimitDegrees=3}
 {bone1Name=dyn_jacket_fr_r_01, bone2Name=spine_05, swing1LimitDegrees=25, swing2LimitDegrees=70, twistLimitDegrees=1}
 {
  bone1Name=spine_03
  bone2Name=spine_02
  swing1LimitDegrees=45
  swing1Motion=Free
  swing2LimitDegrees=45
  swing2Motion=Free
  twistLimitDegrees=45
  twistMotion=Free
 }
 {bone1Name=dyn_jacket_fr_l_02, bone2Name=dyn_jacket_fr_l_01, swing1LimitDegrees=0, swing2LimitDegrees=70, twistLimitDegrees=0}
 {bone1Name=dyn_jacket_fr_r_02, bone2Name=dyn_jacket_fr_r_01, swing1LimitDegrees=0, swing2LimitDegrees=70, twistLimitDegrees=0}
 {bone1Name=dyn_jacket_l_02, bone2Name=dyn_jacket_l_01, swing1LimitDegrees=0, swing2LimitDegrees=70, twistLimitDegrees=0}
 {bone1Name=dyn_jacket_r_02, bone2Name=dyn_jacket_r_01, swing1LimitDegrees=0, swing2LimitDegrees=70, twistLimitDegrees=0}
 {bone1Name=dyn_jacket__bk_r_01, bone2Name=spine_04, swing1LimitDegrees=0, swing2LimitDegrees=60, twistLimitDegrees=0}
 {bone1Name=dyn_jacket__bk_l_01, bone2Name=spine_04, swing1LimitDegrees=0, swing2LimitDegrees=60, twistLimitDegrees=0}
 {
  bone1Name=dyn_jacket_fr_r_02
  bone2Name=dyn_jacket_r_02
  swing1LimitDegrees=45
  swing1Motion=Free
  swing2LimitDegrees=45
  swing2Motion=Free
  twistLimitDegrees=45
  twistMotion=Free
 }
 {
  bone1Name=dyn_jacket_fr_l_02
  bone2Name=dyn_jacket_l_02
  swing1LimitDegrees=45
  swing1Motion=Free
  swing2LimitDegrees=45
  swing2Motion=Free
  twistLimitDegrees=45
  twistMotion=Free
 }
]
