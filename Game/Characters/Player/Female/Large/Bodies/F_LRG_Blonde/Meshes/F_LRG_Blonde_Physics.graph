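# Physics asset for the F_LRG_Blonde large female player body, covering the lower body and a hip tool bag with pouch and hammer.
# Capsules cover the pelvis, both thighs and only the right calf, with angularly free constraints between them.
# Boxes form a dyn_ accessory chain from pelvis through tool bag base and three-box hip pouch to hammer.
# The chain's angular limits keep twist tight at bag and pouch, looser at the hammer.
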PhysicsAsset /Game/Characters/Player/Female/Large/Bodies/F_LRG_Blonde/Meshes/F_LRG_Blonde_Physics
bodies {
 calf_r [{shapeType=Capsule, length=21.6645, radius=11.5072, center=<20.9201 0 -0.2316>, rotation=<rot -89.366 -179.9999 179.9999>}]
 dyn_hammer_swing [{shapeType=Box, extentX=5.0856, extentY=1.5691, extentZ=29.5955, center=<0.2132 -0.1523 -9.3548>, rotation=<rot 2.3666 -90 -0.3944>}]
 dyn_hipPouch [
  {shapeType=Box, extentX=14.6773, extentY=15.695, extentZ=3.4247, center=<7.2495 -0.1447 0.2394>, rotation=<rot 1.5544 -2.783 -4.9962>}
  {shapeType=Box, extentX=16.1384, extentY=4.7636, extentZ=3.1249, center=<7.5971 9.0021 -0.8656>, rotation=<rot -7.0815 -3.1542 24.0824>}
  {shapeType=Box, extentX=16.2535, extentY=7.5, extentZ=4.6356, center=<7.4419 -7.4865 -1.9797>, rotation=<rot -3.634 -0.4649 -23.9213>}
 ]
 dyn_tool_bag_base [{shapeType=Box, extentX=10.1577, extentY=15.0819, extentZ=2.6956, center=<4.938 0 0>}]
 pelvis [{shapeType=Capsule, length=10.1395, radius=16.3829, center=<-4.4193 -0.8691 0>}]
 thigh_l [{shapeType=Capsule, length=16.9215, radius=9.1689, center=<-17.868 -0.9056 -0.5775>, rotation=<rot 87.4454 180 0>}]
 thigh_r [{shapeType=Capsule, length=19.8099, radius=10.7233, center=<19.7573 2.4413 1.5988>, rotation=<rot 90 -0.0001 -179.9999>}]
}
constraints [
 {
  bone1Name=thigh_l
  bone2Name=pelvis
  swing1LimitDegrees=45
  swing1Motion=Free
  swing2LimitDegrees=45
  swing2Motion=Free
  twistLimitDegrees=45
  twistMotion=Free
 }
 {
  bone1Name=thigh_r
  bone2Name=pelvis
  swing1LimitDegrees=45
  swing1Motion=Free
  swing2LimitDegrees=45
  swing2Motion=Free
  twistLimitDegrees=45
  twistMotion=Free
 }
 {bone1Name=dyn_tool_bag_base, bone2Name=pelvis, swing1LimitDegrees=15, swing2LimitDegrees=30, twistLimitDegrees=0}
 {bone1Name=dyn_hipPouch, bone2Name=dyn_tool_bag_base, swing1LimitDegrees=30, swing2LimitDegrees=30, twistLimitDegrees=2}
 {bone1Name=dyn_hammer_swing, bone2Name=dyn_hipPouch, swing1LimitDegrees=30, swing2LimitDegrees=30, twistLimitDegrees=30}
 {
  bone1Name=calf_r
  bone2Name=thigh_r
  swing1LimitDegrees=45
  swing1Motion=Free
  swing2LimitDegrees=45
  swing2Motion=Free
  twistLimitDegrees=45
  twistMotion=Free
 }
]
