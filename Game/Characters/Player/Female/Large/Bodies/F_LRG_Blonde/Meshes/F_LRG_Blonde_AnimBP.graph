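# Post-process style Animation Blueprint for the large female Blonde player body that adds secondary motion and corrective offsets to clothing and gear.
# It copies the parent mesh pose, then chains bone constraints, AnimDynamics, curve-named and crouch-driven corrective transforms, three RigidBody simulations, and bag/backpack/cape adjustments.
# Event graph updates gravity override and crouch alpha.
# During emotes, it disables bag RigidBody simulation and enables pelvis-to-thigh twist-driven bag offset.
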AnimBlueprint /Game/Characters/Player/Female/Large/Bodies/F_LRG_Blonde/Meshes/F_LRG_Blonde_AnimBP
target_skeleton=&/F_LRG_Blonde_Skeleton
graphs [
 AnimGraph {nodes [
   n1 = RigidBody!(
    ComponentPose=Local To Component(LocalPose=Use cached pose 'logic2RDs'())
    Alpha
    OverrideWorldGravity:Vector=Get Gravity Override()
    ExternalForce
   ) -> (Pose) {
    bClampLinearTranslationLimitToRefPose=true
    bOverrideWorldGravity=true
    overridePhysicsAsset=&/F_LRG_Blonde_Physics
    simulationSpace=BaseBoneSpace
    alphaBoolBlend {blendOption=Linear}
    baseBoneRef {boneName=pelvis}
    componentAppliedLinearAccClamp=<3000 3000 3000>
    componentLinearAccScale=<0.3 0.2 0.2>
    componentLinearVelScale=<0.3 0.2 0.2>
    simSpaceSettings {
     maxAngularAcceleration=1000
     maxAngularVelocity=3
     maxLinearAcceleration=700
     maxLinearVelocity=500
     velocityScaleZ=0
     worldAlpha=0.5
     externalLinearDragV=<0.2 0.2>
    }
   }
   n2 = RigidBody(ComponentPose=@n1, Alpha=Get RB alpha() -> (RB alpha:double), OverrideWorldGravity=Get Gravity Override()) {
    bClampLinearTranslationLimitToRefPose=true
    bOverrideWorldGravity=true
    overridePhysicsAsset=&/F_LRG_Blonde_Bag_Physics
    simulationSpace=BaseBoneSpace
    alphaBoolBlend {blendOption=Linear}
    alphaScaleBiasClamp {bInterpResult=true}
    baseBoneRef {boneName=pelvis}
    componentAppliedLinearAccClamp=<3000 0 0>
    componentLinearAccScale=<0.4 0.2 0.2>
    componentLinearVelScale=<0.3 0.2 0.2>
    simSpaceSettings {maxAngularAcceleration=1000, maxAngularVelocity=3, maxLinearAcceleration=700, maxLinearVelocity=500, externalLinearDragV=<0.2 0.2>}
   }
   n3 = RigidBody(ComponentPose=@n2, OverrideWorldGravity=Get Gravity Override()) {
    bClampLinearTranslationLimitToRefPose=true
    bOverrideWorldGravity=true
    overridePhysicsAsset=&/F_LRG_Blonde_FrontJacket_Physics
    simulationSpace=BaseBoneSpace
    alphaBoolBlend {blendOption=Linear}
    baseBoneRef {boneName=spine_01}
    componentAppliedLinearAccClamp=<3000 3000 3000>
    componentLinearAccScale=<0.3 0.2 0.2>
    componentLinearVelScale=<0.3 0.2 0.2>
    simSpaceSettings {
     maxAngularAcceleration=1000
     maxAngularVelocity=3
     maxLinearAcceleration=700
     maxLinearVelocity=500
     velocityScaleZ=0
     worldAlpha=0.2
     externalLinearDragV=<0.2 0.2>
    }
   }
   n4 = "Constraint - dyn_mainbag_Offset (R)"!(ComponentPose=@n3, bAlphaBoolEnabled:bool=True, ConstraintWeights_0=1) -> (Pose) {
    alphaInputType=Bool
    alphaBoolBlend {blendInTime=0.2, blendOption=Linear, blendOutTime=0.2}
    boneToModify {boneName=dyn_mainbag_Offset}
    constraintSetup [{transformType=Rotation, targetBone {boneName=thigh_l}}]
    constraintWeights [1]
   }
   n5 = "Transform (Modify) Bone - Bone: dyn_mainbag_Offset"(ComponentPose=@n4, Alpha=Get pelvis_rot() -> (pelvis_rot:double), Translation="0,0,7", Rotation="0,0,30", Scale="1,1,1") {
    rotationMode=BMM_Additive
    rotationSpace=BCS_BoneSpace
    translationMode=BMM_Additive
    translationSpace=BCS_BoneSpace
    alphaBoolBlend {blendOption=Linear}
    boneToModify {boneName=dyn_mainbag_Offset}
    rotation {roll=30}
    translation {z=7}
   }
   n6 = "Transform (Modify) Bone - Bone: attach_backpack"!(ComponentPose=@n5, bAlphaBoolEnabled:bool=True, Translation="0,-1.1,0", Rotation="0,0,0", Scale="1,1,1") -> (Pose) {
    alphaInputType=Bool
    rotationMode=BMM_Additive
    rotationSpace=BCS_BoneSpace
    translationMode=BMM_Additive
    translationSpace=BCS_BoneSpace
    alphaBoolBlend {blendInTime=0.5, blendOutTime=1}
    boneToModify {boneName=attach_backpack}
    translation {y=-1.1}
   }
   n7 = "Transform (Modify) Bone - Bone: attach_cape"!(ComponentPose=@n6, bAlphaBoolEnabled:bool=True, Translation="0,-2.7,0", Rotation="0,5,0", Scale="1,1,1") -> (Pose) {
    alphaInputType=Bool
    rotationMode=BMM_Additive
    rotationSpace=BCS_BoneSpace
    translationMode=BMM_Additive
    translationSpace=BCS_BoneSpace
    alphaBoolBlend {blendInTime=0.5, blendOutTime=1}
    boneToModify {boneName=attach_cape}
    rotation {yaw=5}
    translation {y=-2.7}
   }
   Output Pose(Result=Component To Local(ComponentPose=@n7)) {layerGroup=None}
   n8 = "Constraint - Forearm_brace_r (P)"(
    ComponentPose=Local To Component(LocalPose=Copy Pose From Mesh() {bCopyCurves=true, bUseAttachedParent=true})
    Alpha=1
    ConstraintWeights_0=1
   ) {
    alphaBoolBlend {blendOption=Linear}
    boneToModify {boneName=Forearm_brace_r}
    constraintSetup [{targetBone {boneName=lowerarm_twist_01_r}}]
    constraintWeights [1]
   }
   n9 = "Constraint - dyn_smallbag_front_Offset (PPPPPPPP)"!(
    ComponentPose=@n8
    Alpha=1
    ConstraintWeights_0=0.0003
    ConstraintWeights_1:float=0.3893
    ConstraintWeights_2:float=0.0432
    ConstraintWeights_3:float=0.0079
    ConstraintWeights_4:float=0.003
    ConstraintWeights_5:float=0.0001
    ConstraintWeights_6:float=0.4312
    ConstraintWeights_7:float=0.1251
   ) -> (Pose) {
    alphaBoolBlend {blendOption=Linear}
    boneToModify {boneName=dyn_smallbag_front_Offset}
    constraintSetup [
     {targetBone {boneName=thigh_twist_01_r}}
     {targetBone {boneName=dyn_bagsBase}}
     {targetBone {boneName=spine_01}}
     {targetBone {boneName=spine_02}}
     {targetBone {boneName=spine_03}}
     {targetBone {boneName=spine_04}}
     {targetBone {boneName=pelvis}}
     {targetBone {boneName=thigh_twist_01_l}}
    ]
    constraintWeights [1,1,1,1,1,1,1,1]
   }
   n10 = "Constraint - dyn_shoulderPad (PPPPPPP)"!(
    ComponentPose=@n9
    Alpha=1
    ConstraintWeights_0=0.1449
    ConstraintWeights_1:float=0.4073
    ConstraintWeights_2:float=0
    ConstraintWeights_3:float=0
    ConstraintWeights_4:float=0
    ConstraintWeights_5:float=0.4073
    ConstraintWeights_6:float=0.0405
   ) -> (Pose) {
    alphaBoolBlend {blendOption=Linear}
    boneToModify {boneName=dyn_shoulderPad}
    constraintSetup [
     {targetBone {boneName=clavicle_l}}
     {targetBone {boneName=upperarm_twist_02_l}}
     {targetBone {boneName=spine_03}}
     {targetBone {boneName=spine_04}}
     {targetBone {boneName=spine_05}}
     {targetBone {boneName=upperarm_twist_01_l}}
     {targetBone {boneName=upperarm_l}}
    ]
    constraintWeights [1,1,1,1,1,1,1]
   }
   n11 = "Constraint - dyn_shells (PPPPPP)"!(
    ComponentPose=@n10
    Alpha=1
    ConstraintWeights_0=0.0182
    ConstraintWeights_1:float=0.1413
    ConstraintWeights_2:float=0.0406
    ConstraintWeights_3:float=0.0192
    ConstraintWeights_4:float=0.7777
    ConstraintWeights_5:float=0.003
   ) -> (Pose) {
    alphaBoolBlend {blendOption=Linear}
    boneToModify {boneName=dyn_shells}
    constraintSetup [
     {perAxis {bX=false}, targetBone {boneName=dyn_bagsBase}}
     {targetBone {boneName=spine_01}}
     {targetBone {boneName=spine_02}}
     {targetBone {boneName=spine_03}}
     {targetBone {boneName=pelvis}}
     {targetBone {boneName=thigh_twist_01_l}}
    ]
    constraintWeights [1,1,1,1,1,1]
   }
   n12 = "Constraint - dyn_kneepad (PP)"!(ComponentPose=@n11, Alpha=1, ConstraintWeights_0=0.7, ConstraintWeights_1:float=0.3) -> (Pose) {
    alphaBoolBlend {blendOption=Linear}
    boneToModify {boneName=dyn_kneepad}
    constraintSetup [{targetBone {boneName=thigh_twist_01_r}}, {targetBone {boneName=calf_r}}]
    constraintWeights [1,1]
   }
   n13 = "Constraint - dyn_collar_l (P)"(ComponentPose=@n12, Alpha=0.5, ConstraintWeights_0=1) {
    alphaBoolBlend {blendOption=Linear}
    boneToModify {boneName=dyn_collar_l}
    constraintSetup [{targetBone {boneName=clavicle_l}}]
    constraintWeights [1]
   }
   n14 = "Constraint - dyn_collar_r (P)"(ComponentPose=@n13, Alpha=0.7, ConstraintWeights_0=1) {
    alphaBoolBlend {blendOption=Linear}
    boneToModify {boneName=dyn_collar_r}
    constraintSetup [{targetBone {boneName=clavicle_r}}]
    constraintWeights [1]
   }
   n15 = "Constraint - dyn_smallbag_front (R)"(ComponentPose=@n14, Alpha="Get Is Crouching "() -> ("Is Crouching ":double), ConstraintWeights_0=1) {
    alphaBoolBlend {blendOption=Linear}
    boneToModify {boneName=dyn_smallbag_front}
    constraintSetup [{transformType=Rotation, targetBone {boneName=thigh_l}}]
    constraintWeights [1]
   }
   Save cached pose 'Constraints2AnimDynam'(Pose=Component To Local(ComponentPose=@n15)) {cacheName=Constraints2AnimDynam}
   n16 = Anim Dynamics - Bone: dyn_front_pouch!(
    ComponentPose=Local To Component(LocalPose=Use cached pose 'Constraints2AnimDynam'())
    Alpha=0.5
    GravityOverride:Vector="0,0,0"
    ExternalForce="0,0,0"
   ) -> (Pose) {
    angularDampingOverride=0.9999
    angularSpringConstant=50
    bAngularSpring=true
    bGravityOverrideInSimSpace=true
    bOverrideAngularDamping=true
    bUseGravityOverride=true
    gravityScale=0.5
    simulationSpace=BoneRelative
    alphaBoolBlend {blendOption=Linear}
    boundBone {boneName=dyn_front_pouch}
    componentAppliedLinearAccClamp=<5000 5000 5000>
    componentLinearAccScale=<2 0.5 2>
    physicsBodyDefinitions [{
      sphereCollisionRadius=0
      boundBone {boneName=dyn_front_pouch}
      boxExtents=<10 10 5>
      constraintSetup {twistAxis=AxisY, angularLimitsMax=<20 20 20>, angularLimitsMin=<-20 -50 -20>, angularTarget=<0 0 0>}
      localJointOffset=<0 0 -1>
     }]
    relativeSpaceBone {boneName=root}
   }
   n17 = Anim Dynamics - Bone: dyn_smallbag_front!(ComponentPose=@n16, Alpha=0.5, GravityOverride:Vector="0,0,0", ExternalForce="0,0,0") -> (Pose) {
    angularDampingOverride=0.9999
    angularSpringConstant=50
    bAngularSpring=true
    bGravityOverrideInSimSpace=true
    bOverrideAngularDamping=true
    bUseGravityOverride=true
    gravityScale=0.5
    simulationSpace=BoneRelative
    alphaBoolBlend {blendOption=Linear}
    boundBone {boneName=dyn_smallbag_front}
    componentAppliedLinearAccClamp=<5000 5000 5000>
    componentLinearAccScale=<2 0.5 2>
    physicsBodyDefinitions [{
      sphereCollisionRadius=0
      boundBone {boneName=dyn_smallbag_front}
      boxExtents=<10 10 5>
      constraintSetup {twistAxis=AxisY, angularLimitsMax=<0 20 20>, angularLimitsMin=<0 -50 -20>, angularTarget=<0 0 0>}
      localJointOffset=<0 0 -1>
     }]
    relativeSpaceBone {boneName=root}
   }
   n18 = Anim Dynamics - Bone: dyn_jacket_bk_01!(ComponentPose=@n17, Alpha=1, GravityOverride:Vector="0,0,0", ExternalForce="0,0,0") -> (Pose) {
    angularDampingOverride=0.9999
    angularSpringConstant=75
    bAngularSpring=true
    bGravityOverrideInSimSpace=true
    bOverrideAngularDamping=true
    bUseGravityOverride=true
    simulationSpace=BoneRelative
    alphaBoolBlend {blendOption=Linear}
    boundBone {boneName=dyn_jacket_bk_01}
    componentAppliedLinearAccClamp=<5000 5000 5000>
    componentLinearAccScale=<0.5 0.5 1>
    componentLinearVelScale=<0.5 0.5 0.5>
    physicsBodyDefinitions [{
      sphereCollisionRadius=0
      boundBone {boneName=dyn_jacket_bk_01}
      boxExtents=<10 20 20>
      constraintSetup {angularLimitsMax=<5 -5 10>, angularLimitsMin=<-5 -25 -10>, angularTarget=<0 0 0>}
      localJointOffset=<2 0 0>
     }]
    relativeSpaceBone {boneName=root}
   }
   n19 = Anim Dynamics - Bone: dyn_jacket__bk_l_01!(ComponentPose=@n18, Alpha=1, GravityOverride:Vector="0,0,0", ExternalForce="0,0,0") -> (Pose) {
    angularDampingOverride=0.9999
    angularSpringConstant=75
    bAngularSpring=true
    bGravityOverrideInSimSpace=true
    bOverrideAngularDamping=true
    bUseGravityOverride=true
    simulationSpace=BoneRelative
    alphaBoolBlend {blendOption=Linear}
    boundBone {boneName=dyn_jacket__bk_l_01}
    componentAppliedLinearAccClamp=<5000 5000 5000>
    componentLinearAccScale=<0.5 0.5 1>
    componentLinearVelScale=<0.5 0.5 0.5>
    physicsBodyDefinitions [{
      sphereCollisionRadius=0
      boundBone {boneName=dyn_jacket__bk_l_01}
      boxExtents=<10 20 20>
      constraintSetup {angularLimitsMax=<5 -5 10>, angularLimitsMin=<-5 -25 -10>, angularTarget=<0 0 0>}
      localJointOffset=<2 0 0>
     }]
    relativeSpaceBone {boneName=root}
   }
   n20 = Anim Dynamics - Bone: dyn_jacket__bk_r_01!(ComponentPose=@n19, Alpha=1, GravityOverride:Vector="0,0,0", ExternalForce="0,0,0") -> (Pose) {
    angularDampingOverride=0.9999
    angularSpringConstant=75
    bAngularSpring=true
    bGravityOverrideInSimSpace=true
    bOverrideAngularDamping=true
    bUseGravityOverride=true
    simulationSpace=BoneRelative
    alphaBoolBlend {blendOption=Linear}
    boundBone {boneName=dyn_jacket__bk_r_01}
    componentAppliedLinearAccClamp=<5000 5000 5000>
    componentLinearAccScale=<0.5 0.5 1>
    componentLinearVelScale=<0.5 0.5 0.5>
    physicsBodyDefinitions [{
      sphereCollisionRadius=0
      boundBone {boneName=dyn_jacket__bk_r_01}
      boxExtents=<10 20 20>
      constraintSetup {angularLimitsMax=<5 -5 10>, angularLimitsMin=<-5 -25 -10>, angularTarget=<0 0 0>}
      localJointOffset=<2 0 0>
     }]
    relativeSpaceBone {boneName=root}
   }
   n21 = Anim Dynamics - Bone: dyn_pliers_r!(ComponentPose=@n20, Alpha=0.2, GravityOverride:Vector="0,0,-980", ExternalForce="0,0,0") -> (Pose) {
    angularDampingOverride=0.99
    angularSpringConstant=5
    bAngularSpring=true
    bGravityOverrideInSimSpace=true
    bLinearSpring=true
    bOverrideAngularDamping=true
    bOverrideLinearDamping=true
    bUseGravityOverride=true
    gravityScale=0.4
    linearDampingOverride=0.99
    linearSpringConstant=5
    alphaBoolBlend {blendOption=Linear}
    boundBone {boneName=dyn_pliers_r}
    componentAppliedLinearAccClamp=<1500 1500 1500>
    componentLinearAccScale=<0.5 0.5 0.5>
    gravityOverride {z=-980}
    physicsBodyDefinitions [{
      sphereCollisionRadius=0
      boundBone {boneName=dyn_pliers_r}
      boxExtents=<30 30 10>
      constraintSetup {angularLimitsMax=<124.6752 0 0>, angularLimitsMin=<-1.4421 0 0>, angularTarget=<0 0 0>, linearAxesMax=<0 10 0>}
      localJointOffset=<0 0 -10>
     }]
   }
   n22 = Anim Dynamics - Bone: dyn_handle_f!(ComponentPose=@n21, Alpha=0.3, GravityOverride:Vector="0,0,-980", ExternalForce="0,0,0") -> (Pose) {
    angularDampingOverride=0.99
    angularSpringConstant=5
    bAngularSpring=true
    bGravityOverrideInSimSpace=true
    bLinearSpring=true
    bOverrideAngularDamping=true
    bOverrideLinearDamping=true
    bUseGravityOverride=true
    gravityScale=0.3
    linearDampingOverride=0.99
    linearSpringConstant=5
    alphaBoolBlend {blendOption=Linear}
    boundBone {boneName=dyn_handle_f}
    componentAppliedLinearAccClamp=<50 50 50>
    componentLinearAccScale=<0.2 0.2>
    componentLinearVelScale=<0.2 0.2>
    gravityOverride {z=-980}
    physicsBodyDefinitions [{
      sphereCollisionRadius=0
      boundBone {boneName=dyn_handle_f}
      boxExtents=<15 15 5>
      constraintSetup {angularLimitsMax=<0 44.5008 0>, angularLimitsMin=<0 0 0>, angularTarget=<0 0 0>, linearAxesMax=<5 2 0>, linearAxesMin=<0 -1 0>}
      localJointOffset=<0 0 -3>
     }]
   }
   n23 = Anim Dynamics - Bone: dyn_screwdriver_flat_f!(ComponentPose=@n22, Alpha=0.3, GravityOverride:Vector="0,0,-980", ExternalForce="0,0,0") -> (Pose) {
    angularDampingOverride=0.99
    angularSpringConstant=10
    bGravityOverrideInSimSpace=true
    bOverrideAngularDamping=true
    bOverrideLinearDamping=true
    bUseGravityOverride=true
    linearDampingOverride=0.99
    alphaBoolBlend {blendOption=Linear}
    boundBone {boneName=dyn_screwdriver_flat_f}
    componentAppliedLinearAccClamp=<2500 2500 2500>
    componentLinearAccScale=<0.2 0.2>
    gravityOverride {z=-980}
    physicsBodyDefinitions [{
      sphereCollisionRadius=0
      boundBone {boneName=dyn_screwdriver_flat_f}
      boxExtents=<5 5 5>
      constraintSetup {angularLimitsMax=<1 5 0>, angularLimitsMin=<-1 -5 0>, angularTarget=<0 0 0>, linearAxesMax=<3 5 0>, linearAxesMin=<-3 -5 0>}
      localJointOffset=<0 0 -3>
     }]
   }
   n24 = Anim Dynamics - Bone: dyn_screwDriver_flat_r!(ComponentPose=@n23, Alpha=0.2, GravityOverride:Vector="0,0,-980", ExternalForce="0,0,0") -> (Pose) {
    angularDampingOverride=0.99
    angularSpringConstant=75
    bAngularSpring=true
    bGravityOverrideInSimSpace=true
    bLinearSpring=true
    bOverrideAngularDamping=true
    bOverrideLinearDamping=true
    bUseGravityOverride=true
    gravityScale=0.3
    linearDampingOverride=0.99
    linearSpringConstant=75
    alphaBoolBlend {blendOption=Linear}
    boundBone {boneName=dyn_screwDriver_flat_r}
    componentAppliedLinearAccClamp=<50 50 50>
    componentLinearAccScale=<0.2 0.2>
    componentLinearVelScale=<0.2 0.2>
    gravityOverride {z=-980}
    physicsBodyDefinitions [{
      sphereCollisionRadius=0
      boundBone {boneName=dyn_screwDriver_flat_r}
      boxExtents=<15 15 5>
      constraintSetup {angularLimitsMax=<0 0 0>, angularLimitsMin=<-71.8858 0 0>, angularTarget=<0 0 0>, linearAxesMin=<0 -3 0>}
      localJointOffset=<0 0 -3>
     }]
   }
   n25 = Anim Dynamics - Bone: dyn_screwdriver_philips_f!(ComponentPose=@n24, Alpha=0.3, GravityOverride:Vector="0,0,-980", ExternalForce="0,0,0") -> (Pose) {
    angularDampingOverride=0.99
    angularSpringConstant=500
    bGravityOverrideInSimSpace=true
    bLinearSpring=true
    bOverrideAngularDamping=true
    bOverrideLinearDamping=true
    bUseGravityOverride=true
    gravityScale=0.3
    linearDampingOverride=0.99
    linearSpringConstant=20
    alphaBoolBlend {blendOption=Linear}
    boundBone {boneName=dyn_screwdriver_philips_f}
    componentAppliedLinearAccClamp=<50 50 50>
    componentLinearAccScale=<0.2 0.2>
    componentLinearVelScale=<0.2 0.2>
    gravityOverride {z=-980}
    physicsBodyDefinitions [{
      sphereCollisionRadius=0
      boundBone {boneName=dyn_screwdriver_philips_f}
      boxExtents=<15 15 5>
      constraintSetup {angularLimitsMax=<1 0 0>, angularLimitsMin=<-1 -32.3851 0>, angularTarget=<0 0 0>, linearAxesMax=<0 2 0>, linearAxesMin=<-3 -1 0>}
      localJointOffset=<0 0 -3>
     }]
   }
   Save cached pose 'AnimDynam2Logic'(Pose=Component To Local(ComponentPose=@n25)) {cacheName=AnimDynam2Logic}
   n26 = "Transform (Modify) Bone - Bone: dyn_shoulderPad_offset"(
    ComponentPose=Local To Component(LocalPose=Use cached pose 'AnimDynam2Logic'())
    Alpha=Get LookUpRCollar() -> (LookUpRCollar:float)
    Translation="4,0,2"
    Rotation="0,0,0"
    Scale="0,0,0"
   ) {
    rotationMode=BMM_Additive
    rotationSpace=BCS_BoneSpace
    scaleSpace=BCS_BoneSpace
    translationMode=BMM_Additive
    translationSpace=BCS_BoneSpace
    alphaBoolBlend {blendOption=Linear}
    boneToModify {boneName=dyn_shoulderPad_offset}
    scale=<0 0 0>
    translation {x=4, z=2}
   }
   n27 = "Transform (Modify) Bone - Bone: dyn_collar_r"(ComponentPose=@n26, Alpha=Get LookUpRCollar(), Translation="0,0,0", Rotation="-12,-35.527,-10.243", Scale="1,1,1") {
    rotationMode=BMM_Additive
    scaleMode=BMM_Additive
    translationMode=BMM_Additive
    alphaBoolBlend {blendOption=Linear}
    boneToModify {boneName=dyn_collar_r}
    rotation=<rot -12 -35.527 -10.243>
   }
   n28 = "Transform (Modify) Bone - Bone: dyn_chest_push_in_r"(ComponentPose=@n27, Alpha=Get ChestPushInR() -> (ChestPushInR:float), Translation="9,0.000,0.000", Rotation="0,0,0", Scale="1,1,1") {
    rotationMode=BMM_Additive
    scaleMode=BMM_Additive
    translationMode=BMM_Additive
    alphaBoolBlend {blendOption=Linear}
    boneToModify {boneName=dyn_chest_push_in_r}
    translation {x=9}
   }
   n29 = "Transform (Modify) Bone - Bone: dyn_chest_push_in_r"(ComponentPose=@n28, Alpha=Get ChestPushInRCorr() -> (ChestPushInRCorr:float), Translation="-9,0,0", Rotation="0,-30,0", Scale="1,1,1") {
    rotationMode=BMM_Additive
    scaleMode=BMM_Additive
    translationMode=BMM_Additive
    alphaBoolBlend {blendOption=Linear}
    boneToModify {boneName=dyn_chest_push_in_r}
    rotation {yaw=-30}
    translation {x=-9}
   }
   n30 = "Transform (Modify) Bone - Bone: dyn_collar_back_r"(ComponentPose=@n29, Alpha=Get CollarPopRight() -> (CollarPopRight:float), Translation="0,0,0", Rotation="0,0,0", Scale="1,1,1") {
    rotationMode=BMM_Additive
    scaleMode=BMM_Additive
    translationMode=BMM_Additive
    alphaBoolBlend {blendOption=Linear}
    boneToModify {boneName=dyn_collar_back_r}
   }
   n31 = "Transform (Modify) Bone - Bone: dyn_collar_l"(ComponentPose=@n30, Alpha=Get ExtraLiftLCollar() -> (ExtraLiftLCollar:float), Translation="0,0,0", Rotation="20,0,0", Scale="1,1,1") {
    rotationMode=BMM_Additive
    scaleMode=BMM_Additive
    translationMode=BMM_Additive
    alphaBoolBlend {blendOption=Linear}
    boneToModify {boneName=dyn_collar_l}
    rotation {pitch=20}
   }
   n32 = Get CollarPop() -> (CollarPop:float)
   n33 = "Transform (Modify) Bone - Bone: dyn_collar_back_l"(ComponentPose=@n31, Alpha=@n32, Translation="0,0,-1", Rotation="-10,0,0", Scale="1,1,1") {
    rotationMode=BMM_Additive
    rotationSpace=BCS_BoneSpace
    scaleMode=BMM_Additive
    scaleSpace=BCS_BoneSpace
    translationMode=BMM_Additive
    translationSpace=BCS_BoneSpace
    alphaBoolBlend {blendOption=Linear}
    boneToModify {boneName=dyn_collar_back_l}
    rotation {pitch=-10}
    translation {z=-1}
   }
   n34 = "Transform (Modify) Bone - Bone: dyn_collar_back"(ComponentPose=@n33, Alpha=@n32, Translation="0,0,-2.5", Rotation="-10,0,0", Scale="1,1,1") {
    rotationMode=BMM_Additive
    rotationSpace=BCS_BoneSpace
    scaleMode=BMM_Additive
    scaleSpace=BCS_BoneSpace
    translationMode=BMM_Additive
    translationSpace=BCS_BoneSpace
    alphaBoolBlend {blendOption=Linear}
    boneToModify {boneName=dyn_collar_back}
    rotation {pitch=-10}
    translation {z=-2.5}
   }
   n35 = "Transform (Modify) Bone - Bone: dyn_collar_back_r"(ComponentPose=@n34, Alpha=@n32, Translation="0,0,-2", Rotation="-10,0,0", Scale="1,1,1") {
    rotationMode=BMM_Additive
    rotationSpace=BCS_BoneSpace
    scaleMode=BMM_Additive
    scaleSpace=BCS_BoneSpace
    translationMode=BMM_Additive
    translationSpace=BCS_BoneSpace
    alphaBoolBlend {blendOption=Linear}
    boneToModify {boneName=dyn_collar_back_r}
    rotation {pitch=-10}
    translation {z=-2}
   }
   n36 = "Transform (Modify) Bone - Bone: dyn_collar_back"(ComponentPose=@n35, Alpha=@n32, Translation="0,-2,2", Rotation="0,0,0", Scale="1,1,1") {
    rotationMode=BMM_Additive
    scaleMode=BMM_Additive
    translationMode=BMM_Additive
    alphaBoolBlend {blendOption=Linear}
    boneToModify {boneName=dyn_collar_back}
    translation {y=-2, z=2}
   }
   n37 = "Transform (Modify) Bone - Bone: dyn_shoulderPad_offset"(ComponentPose=@n36, Alpha=Get LookUpLCollar() -> (LookUpLCollar:float), Translation="0,0,0.2", Rotation="0,0,0", Scale="0,0,0") {
    rotationMode=BMM_Additive
    rotationSpace=BCS_BoneSpace
    scaleSpace=BCS_BoneSpace
    translationMode=BMM_Additive
    translationSpace=BCS_BoneSpace
    alphaBoolBlend {blendOption=Linear}
    boneToModify {boneName=dyn_shoulderPad_offset}
    scale=<0 0 0>
    translation {z=0.2}
   }
   n38 = "Transform (Modify) Bone - Bone: dyn_bootStrap_r"(ComponentPose=@n37, Alpha=Get FootRotate() -> (FootRotate:float), Translation="0,0,0", Rotation="-5,-5,-20", Scale="1,1,1") {
    rotationMode=BMM_Additive
    scaleMode=BMM_Additive
    translationMode=BMM_Additive
    alphaBoolBlend {blendOption=Linear}
    boneToModify {boneName=dyn_bootStrap_r}
    rotation=<rot -5 -5 -20>
   }
   n39 = "Transform (Modify) Bone - Bone: dyn_collar_back_l"(ComponentPose=@n38, Alpha=0.5, Translation="0,0,-1", Rotation="-10,0,0", Scale="1,1,1") {
    rotationMode=BMM_Additive
    rotationSpace=BCS_BoneSpace
    scaleMode=BMM_Additive
    scaleSpace=BCS_BoneSpace
    translationMode=BMM_Additive
    translationSpace=BCS_BoneSpace
    alphaBoolBlend {blendOption=Linear}
    boneToModify {boneName=dyn_collar_back_l}
    rotation {pitch=-10}
    translation {z=-1}
   }
   n40 = "Transform (Modify) Bone - Bone: dyn_collar_back"(ComponentPose=@n39, Alpha=0.5, Translation="0,0,-2.5", Rotation="-10,0,0", Scale="1,1,1") {
    rotationMode=BMM_Additive
    rotationSpace=BCS_BoneSpace
    scaleMode=BMM_Additive
    scaleSpace=BCS_BoneSpace
    translationMode=BMM_Additive
    translationSpace=BCS_BoneSpace
    alphaBoolBlend {blendOption=Linear}
    boneToModify {boneName=dyn_collar_back}
    rotation {pitch=-10}
    translation {z=-2.5}
   }
   n41 = "Transform (Modify) Bone - Bone: dyn_collar_back_r"(ComponentPose=@n40, Alpha=0.5, Translation="0,0,-2", Rotation="-10,0,0", Scale="1,1,1") {
    rotationMode=BMM_Additive
    rotationSpace=BCS_BoneSpace
    scaleMode=BMM_Additive
    scaleSpace=BCS_BoneSpace
    translationMode=BMM_Additive
    translationSpace=BCS_BoneSpace
    alphaBoolBlend {blendOption=Linear}
    boneToModify {boneName=dyn_collar_back_r}
    rotation {pitch=-10}
    translation {z=-2}
   }
   n42 = "Transform (Modify) Bone - Bone: dyn_shoulderPad_offset"(
    ComponentPose=@n41
    Alpha=Get IsInFrontEndAlpha() -> (IsInFrontEndAlpha:float)
    Translation="0,-2,0.2"
    Rotation="15,-15,-15"
    Scale="1.05,1.05,1.05"
   ) {
    rotationMode=BMM_Additive
    scaleMode=BMM_Additive
    translationMode=BMM_Additive
    alphaBoolBlend {blendOption=Linear}
    boneToModify {boneName=dyn_shoulderPad_offset}
    rotation=<rot 15 -15 -15>
    scale=<1.05 1.05 1.05>
    translation {y=-2, z=0.2}
   }
   n43 = "Get Is Crouching "()
   n44 = "Transform (Modify) Bone - Bone: dyn_shells_Offset"(ComponentPose=@n42, Alpha=@n43, Translation="0.000,1.317,1.000", Rotation="10.591,1.096,-7.883", Scale="1,1,1") {
    rotationMode=BMM_Additive
    scaleMode=BMM_Additive
    translationMode=BMM_Additive
    alphaBoolBlend {blendOption=Linear}
    boneToModify {boneName=dyn_shells_Offset}
    rotation=<rot 10.591 1.096 -7.883>
    translation {y=1.317, z=1}
   }
   n45 = "Transform (Modify) Bone - Bone: dyn_smallbag_front_Offset"(ComponentPose=@n44, Alpha=@n43, Translation="0.056,-3.017,4.408", Rotation="10.728,-5.356,77.612", Scale="1,1,1") {
    rotationMode=BMM_Additive
    scaleMode=BMM_Additive
    scaleSpace=BCS_BoneSpace
    translationMode=BMM_Additive
    alphaBoolBlend {blendOption=Linear}
    boneToModify {boneName=dyn_smallbag_front_Offset}
    rotation=<rot 10.728 -5.356 77.612>
    translation=<0.056 -3.017 4.408>
   }
   n46 = "Transform (Modify) Bone - Bone: dyn_mainbag_Offset"(ComponentPose=@n45, Alpha=@n43, Translation="0.895,-1.250,0.000", Rotation="-0.000,-14.103,0.000", Scale="1,1,1") {
    rotationMode=BMM_Additive
    scaleMode=BMM_Additive
    scaleSpace=BCS_BoneSpace
    translationMode=BMM_Additive
    alphaBoolBlend {blendOption=Linear}
    boneToModify {boneName=dyn_mainbag_Offset}
    rotation {yaw=-14.103}
    translation=<0.895 -1.25>
   }
   Save cached pose 'logic2RDs'(Pose=Component To Local(ComponentPose=@n46)) {cacheName=logic2RDs}
  ]}
 EventGraph {nodes [
   Event BlueprintUpdateAnimation()
   > n1 = Sequence()
   > n2 = Gravity Override State Function DEPRECATED(CharacterPartAnimInstance=Self-Reference(), GameTimeInSeconds=GetGameTimeInSeconds(), GravityOverrideParams=Get Gravity Override Params())
   > Set Gravity Override(Gravity Override=@n2)
   > "Set Is Crouching "("Is Crouching ":double="To Float (Boolean)"(InBool=Get bIsCrouching())) -> (Output_Get:double)
   > Set pelvis_rot(pelvis_rot:double=GetAlphaFromReferenceTransformDelta(SkeletalMeshComponent=Get SkelMesh(), BoneName=pelvis, BaseName=thigh_l, InputRange=(X=0,Y=150), TwistAxis="0,0,1")) -> (Output_Get:double)
   > n3 = Branch(Condition=Get bIsPlayingEmote())
   > Set RB alpha(RB alpha:double=0) -> (Output_Get:double)
   > Set pelvis_rot(pelvis_rot=Get pelvis_rot())
   Set RB alpha(execute=@n3.else, RB alpha=1)
   > Set pelvis_rot(pelvis_rot=0)
   n4 = Get bIsTacticalSprinting()
   n5 = Get bIsJumping()
   Set ExtraBackpackAdjust(execute=@n1.then_1, ExtraBackpackAdjust:bool=OR Boolean(A=AND Boolean(A=@n4, B=@n5), B=Get bIsSkydiveFloating())) -> (Output_Get:bool)
   n6 = Event OnCharacterCustomizationLoaded()
   > Set SkelMesh(SkelMesh=@n6.CustomBaseMesh)
  ]}
]
variables [
 ExtraBackpackAdjust {type=boolean}
 Gravity Override {type=Vector}
 Gravity Override Params {type=GravityOverrideParamsStruct, default {ziplining {gravityOverride=<0 0 -200>}}}
 "Is Crouching " {type=number}
 pelvis_rot {type=number}
 RB alpha {type=number}
 SkelMesh {type=SkeletalMeshComponent}
]
