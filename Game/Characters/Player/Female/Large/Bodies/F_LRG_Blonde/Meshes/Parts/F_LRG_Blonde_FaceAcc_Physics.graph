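# Physics asset for the FaceAcc mesh of the large blonde female player character, defining dynamic hair chains and upper-body shapes.
# Capsules cover the spine, clavicles, neck and head, joined by free constraints, alongside uniformly sized capsules on each hair bone.
# Hair chains mostly hang from the head with limited swing and twist, narrower sideways on center and right back strands.
# Front-left branches off dyn_hair_l_02, whose link to dyn_hair_l_03 rotates freely.
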PhysicsAsset /Game/Characters/Player/Female/Large/Bodies/F_LRG_Blonde/Meshes/Parts/F_LRG_Blonde_FaceAcc_Physics
bodies {
 clavicle_l [{shapeType=Capsule, length=7.2935, radius=7.3987, center=<5.5922 -0.9147 -1.9537>, rotation=<rot 83.3944 0 0>}]
 clavicle_r [{shapeType=Capsule, length=12.1597, radius=9.0375, center=<-4.0137 -1.4754 7.0499>, rotation=<rot 83.3934 0 0.3102>}]
 dyn_hair_01 [{shapeType=Capsule, length=5, radius=2, shapeName=dyn_hair_01_capsule, center=<1.3511 0 0>}]
 dyn_hair_02 [{shapeType=Capsule, length=5, radius=2, shapeName=dyn_hair_02_capsule, center=<1.3511 0 0>}]
 dyn_hair_03 [{shapeType=Capsule, length=5, radius=2, shapeName=dyn_hair_03_capsule, center=<1.3511 0 0>}]
 dyn_hair_04 [{shapeType=Capsule, length=5, radius=2, shapeName=dyn_hair_04_capsule, center=<1.3511 0 0>}]
 dyn_hair_05 [{shapeType=Capsule, length=5, radius=2, shapeName=dyn_hair_05_capsule, center=<1.3511 0 0>}]
 dyn_hair_08_a [{shapeType=Capsule, length=5, radius=2, shapeName=dyn_hair_08_a_capsule, center=<1.3511 0 0>}]
 dyn_hair_08_b [{shapeType=Capsule, length=5, radius=2, shapeName=dyn_hair_08_b_capsule, center=<1.3511 0 0>}]
 dyn_hair_08_c [{shapeType=Capsule, length=5, radius=2, shapeName=dyn_hair_08_c_capsule, center=<1.3511 0 0>, rotation=<rot 0 15.6187 0>}]
 dyn_hair_bk_01 [{shapeType=Capsule, length=5, radius=2, shapeName=dyn_hair_bk_01_capsule, center=<1.3511 0 0>}]
 dyn_hair_bk_02 [{shapeType=Capsule, length=5, radius=2, shapeName=dyn_hair_bk_02_capsule, center=<1.3511 0 0>}]
 dyn_hair_bk_03 [{shapeType=Capsule, length=5, radius=2, shapeName=dyn_hair_bk_03_capsule, center=<1.3511 0 0>}]
 dyn_hair_bk_l_01 [{shapeType=Capsule, length=5, radius=2, shapeName=dyn_hair_bk_l_01_capsule, center=<1.3511 0 0>}]
 dyn_hair_bk_l_02 [{shapeType=Capsule, length=5, radius=2, shapeName=dyn_hair_bk_l_02_capsule, center=<1.3511 0 0>}]
 dyn_hair_bk_l_03 [{shapeType=Capsule, length=5, radius=2, shapeName=dyn_hair_bk_l_03_capsule, center=<1.3511 0 0>}]
 dyn_hair_bk_r_01 [{shapeType=Capsule, length=5, radius=2, shapeName=dyn_hair_bk_r_01_capsule, center=<1.3511 0 0>}]
 dyn_hair_bk_r_02 [{shapeType=Capsule, length=5, radius=2, shapeName=dyn_hair_bk_r_02_capsule, center=<1.3511 0 0>}]
 dyn_hair_fr_r_01 [{shapeType=Capsule, length=5, radius=2, shapeName=dyn_hair_fr_r_01_capsule, center=<1.3511 0 0>}]
 dyn_hair_fr_r_02 [{shapeType=Capsule, length=5, radius=2, shapeName=dyn_hair_fr_r_02_capsule, center=<1.3511 0 0>}]
 dyn_hair_fr_r_03 [{shapeType=Capsule, length=5, radius=2, shapeName=dyn_hair_fr_r_03_capsule, center=<1.3511 0 0>}]
 dyn_hair_front_l_01 [{shapeType=Capsule, length=5, radius=2, shapeName=dyn_hair_front_l_01_capsule, center=<1.3511 0 0>}]
 dyn_hair_front_l_02 [{shapeType=Capsule, length=5, radius=2, shapeName=dyn_hair_front_l_02_capsule, center=<1.3511 0 0>}]
 dyn_hair_l_01 [{shapeType=Capsule, length=5, radius=2, shapeName=dyn_hair_l_01_capsule, center=<1.3511 0 0>}]
 dyn_hair_l_02 [{shapeType=Capsule, length=5, radius=2, shapeName=dyn_hair_l_02_capsule, center=<1.3511 0 0>}]
 dyn_hair_l_03 [{shapeType=Capsule, length=5, radius=2, shapeName=dyn_hair_l_03_capsule, center=<1.3511 0 0>, rotation=<rot 0 23.1589 0>}]
 dyn_hair_r_01 [{shapeType=Capsule, length=5, radius=2, shapeName=dyn_hair_r_01_capsule, center=<1.3511 0 0>}]
 dyn_hair_r_02 [{shapeType=Capsule, length=5, radius=2, shapeName=dyn_hair_r_02_capsule, center=<1.3511 0 0>}]
 dyn_hair_r_03 [{shapeType=Capsule, length=5, radius=2, shapeName=dyn_hair_r_03_capsule, center=<1.3511 0 0>, rotation=<rot 0 28.9668 0>}]
 head [{shapeType=Capsule, length=5.9657, radius=5.522, center=<0.8118 1.1771 0>, rotation=<rot 90 0 0>}]
 neck_01 [{shapeType=Capsule, length=9.7608, radius=5.6194, center=<-5.3297 -1.8201 0>}]
 neck_02 [
  {shapeType=Capsule, length=7.0097, radius=6.6335, rotation=<rot 90 0 0>}
  {shapeType=Capsule, length=5.8588, radius=5.4787, center=<-2.5218 -1.4907 3.527>, rotation=<rot -52.4061 -94.1401 104.2946>}
 ]
 spine_05 [{shapeType=Capsule, length=13.3729, radius=9.7248, center=<2.6902 -2.7405 0>}]
}
constraints [
 {bone1Name=dyn_hair_r_02, bone2Name=dyn_hair_r_01, swing1LimitDegrees=45, swing2LimitDegrees=45, twistLimitDegrees=5}
 {bone1Name=dyn_hair_r_03, bone2Name=dyn_hair_r_02, swing1LimitDegrees=45, swing2LimitDegrees=45, twistLimitDegrees=5}
 {bone1Name=dyn_hair_bk_r_02, bone2Name=dyn_hair_bk_r_01, swing1LimitDegrees=50, swing2LimitDegrees=10, twistLimitDegrees=10}
 {bone1Name=dyn_hair_l_02, bone2Name=dyn_hair_l_01, swing1LimitDegrees=45, swing2LimitDegrees=45, twistLimitDegrees=5}
 {bone1Name=dyn_hair_bk_02, bone2Name=dyn_hair_bk_01, swing1LimitDegrees=50, swing2LimitDegrees=10, twistLimitDegrees=25}
 {bone1Name=dyn_hair_bk_03, bone2Name=dyn_hair_bk_02, swing1LimitDegrees=50, swing2LimitDegrees=10, twistLimitDegrees=25}
 {bone1Name=dyn_hair_bk_l_02, bone2Name=dyn_hair_bk_l_01, swing1LimitDegrees=45, swing2LimitDegrees=45, twistLimitDegrees=5}
 {bone1Name=dyn_hair_bk_l_03, bone2Name=dyn_hair_bk_l_02, swing1LimitDegrees=45, swing2LimitDegrees=45, twistLimitDegrees=5}
 {bone1Name=dyn_hair_01, bone2Name=head, swing1LimitDegrees=45, swing2LimitDegrees=45, twistLimitDegrees=5}
 {bone1Name=dyn_hair_02, bone2Name=head, swing1LimitDegrees=45, swing2LimitDegrees=45, twistLimitDegrees=5}
 {bone1Name=dyn_hair_03, bone2Name=head, swing1LimitDegrees=45, swing2LimitDegrees=45, twistLimitDegrees=5}
 {bone1Name=dyn_hair_04, bone2Name=head, swing1LimitDegrees=45, swing2LimitDegrees=45, twistLimitDegrees=5}
 {bone1Name=dyn_hair_05, bone2Name=head, swing1LimitDegrees=45, swing2LimitDegrees=45, twistLimitDegrees=5}
 {bone1Name=dyn_hair_r_01, bone2Name=head, swing1LimitDegrees=45, swing2LimitDegrees=45, twistLimitDegrees=5}
 {bone1Name=dyn_hair_bk_r_01, bone2Name=head, swing1LimitDegrees=50, swing2LimitDegrees=10, twistLimitDegrees=10}
 {bone1Name=dyn_hair_l_01, bone2Name=head, swing1LimitDegrees=45, swing2LimitDegrees=45, twistLimitDegrees=5}
 {bone1Name=dyn_hair_bk_01, bone2Name=head, swing1LimitDegrees=50, swing2LimitDegrees=10, twistLimitDegrees=25}
 {bone1Name=dyn_hair_bk_l_01, bone2Name=head, swing1LimitDegrees=45, swing2LimitDegrees=45, twistLimitDegrees=5}
 {
  bone1Name=head
  bone2Name=neck_02
  swing1LimitDegrees=45
  swing1Motion=Free
  swing2LimitDegrees=45
  swing2Motion=Free
  twistLimitDegrees=45
  twistMotion=Free
 }
 {
  bone1Name=neck_02
  bone2Name=neck_01
  swing1LimitDegrees=45
  swing1Motion=Free
  swing2LimitDegrees=45
  swing2Motion=Free
  twistLimitDegrees=45
  twistMotion=Free
 }
 {
  bone1Name=clavicle_r
  bone2Name=spine_05
  swing1LimitDegrees=45
  swing1Motion=Free
  swing2LimitDegrees=45
  swing2Motion=Free
  twistLimitDegrees=45
  twistMotion=Free
 }
 {
  bone1Name=neck_01
  bone2Name=spine_05
  swing1LimitDegrees=45
  swing1Motion=Free
  swing2LimitDegrees=45
  swing2Motion=Free
  twistLimitDegrees=45
  twistMotion=Free
 }
 {bone1Name=dyn_hair_fr_r_01, bone2Name=head, swing1LimitDegrees=45, swing2LimitDegrees=45, twistLimitDegrees=5}
 {bone1Name=dyn_hair_fr_r_02, bone2Name=dyn_hair_fr_r_01, swing1LimitDegrees=45, swing2LimitDegrees=45, twistLimitDegrees=5}
 {
  bone1Name=dyn_hair_l_03
  bone2Name=dyn_hair_l_02
  swing1LimitDegrees=45
  swing1Motion=Free
  swing2LimitDegrees=45
  swing2Motion=Free
  twistLimitDegrees=45
  twistMotion=Free
 }
 {bone1Name=dyn_hair_08_a, bone2Name=head, swing1LimitDegrees=45, swing2LimitDegrees=45, twistLimitDegrees=5}
 {bone1Name=dyn_hair_08_b, bone2Name=dyn_hair_08_a, swing1LimitDegrees=45, swing2LimitDegrees=45, twistLimitDegrees=5}
 {bone1Name=dyn_hair_08_c, bone2Name=dyn_hair_08_b, swing1LimitDegrees=45, swing2LimitDegrees=45, twistLimitDegrees=5}
 {bone1Name=dyn_hair_fr_r_03, bone2Name=dyn_hair_fr_r_02, swing1LimitDegrees=45, swing2LimitDegrees=45, twistLimitDegrees=5}
 {
  bone1Name=clavicle_l
  bone2Name=spine_05
  swing1LimitDegrees=45
  swing1Motion=Free
  swing2LimitDegrees=45
  swing2Motion=Free
  twistLimitDegrees=45
  twistMotion=Free
 }
 {bone1Name=dyn_hair_front_l_01, bone2Name=dyn_hair_l_02, swing1LimitDegrees=45, swing2LimitDegrees=45, twistLimitDegrees=5}
 {bone1Name=dyn_hair_front_l_02, bone2Name=dyn_hair_front_l_01, swing1LimitDegrees=45, swing2LimitDegrees=45, twistLimitDegrees=5}
]
