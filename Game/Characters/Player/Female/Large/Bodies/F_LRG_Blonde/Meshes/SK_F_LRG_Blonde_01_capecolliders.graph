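# Physics asset for the large blonde female player body, named as the collider set that capes collide against.
# Capsules approximate head, neck, spine, pelvis, arms and legs, mostly on twist bones for the limbs, with two extra offset spheres on the pelvis.
# Left and right limbs are not mirrored, differing in capsule sizes and in the forearm bone used.
# Constraints link limb bodies directly to spine_05 or pelvis, while the spine-neck-head chain mostly has free angular motion.
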PhysicsAsset /Game/Characters/Player/Female/Large/Bodies/F_LRG_Blonde/Meshes/SK_F_LRG_Blonde_01_capecolliders
bodies {
 calf_twist_02_l [{shapeType=Capsule, length=41.2069, radius=7.7282, center=<-2.3286 -1.1138 -0.1616>, rotation=<rot -0.0062 87.9534 -89.9636>}]
 calf_twist_02_r [{shapeType=Capsule, length=34.1258, radius=13.5433, center=<-2.362 0.5687 0.1066>, rotation=<rot -0.0147 -90.8481 -89.8927>}]
 foot_l [{shapeType=Capsule, length=13.1659, radius=7.799, center=<-2.495 5.8526 -0.0561>, rotation=<rot -0.672 179.2464 -91.4399>}]
 foot_r [{shapeType=Capsule, length=8.4159, radius=8.7988, center=<2.7493 -5.7525 0.7073>, rotation=<rot 0.6234 1.1655 -88.5384>}]
 head [{shapeType=Capsule, length=2.0719, radius=17.5, center=<13.9496 -0.0001 0>, rotation=<rot 0 -90 0>}]
 lowerarm_l [{shapeType=Capsule, length=35.0482, radius=5.8186, center=<17.5656 -2.5798 -1.2337>, rotation=<rot 2.7686 -102.2653 -86.1165>}]
 lowerarm_twist_01_r [{shapeType=Capsule, length=26.7452, radius=5.8107, center=<-0.8712 0.5689 1.0515>, rotation=<rot 0 -94.0739 -84.6099>}]
 neck_01 [{shapeType=Capsule, length=4.3196, radius=8.1138, rotation=<rot -0.0001 -89.9998 90.3863>}]
 neck_02 [{shapeType=Capsule, length=0.0004, radius=8.0807, rotation=<rot 0 -90 0>}]
 pelvis [
  {shapeType=Sphere, radius=7.5, center=<1.2543 -16.2433 -15.4102>}
  {shapeType=Sphere, radius=9.7124, center=<6.9405 -10.8515 -12.8792>}
  {shapeType=Capsule, length=33, radius=17.8423, center=<-0.9839 -2.4871 -0.4806>, rotation=<rot -4.9681 -82.8847 -0.5913>}
 ]
 spine_01 [{shapeType=Capsule, length=24.9724, radius=11.882, center=<0 -4.2307 0>, rotation=<rot 0 -90 0>}]
 spine_03 [{shapeType=Capsule, length=10.5022, radius=15.3604, center=<0 2.6696 0.0001>}]
 spine_04 [{shapeType=Capsule, length=7.2641, radius=15.6748, center=<0 1.8442 0>}]
 spine_05 [{shapeType=Capsule, length=20.0943, radius=13.5568}]
 thigh_twist_01_l [{shapeType=Capsule, length=30.3336, radius=11.7673, center=<7.9097 -2.3933 -2.0262>, rotation=<rot 0.1146 -82.2321 -88.8882>}]
 thigh_twist_01_r [{shapeType=Capsule, length=22.9493, radius=11.0324, center=<-8.9821 -0.1471 2.2298>, rotation=<rot -1.6368 -85.8033 -85.0445>}]
 upperarm_twist_01_l [{shapeType=Capsule, length=20.9787, radius=6.5, center=<3.8306 -0.4087 0.4242>, rotation=<rot 0 -80 -90>}]
 upperarm_twist_01_r [{shapeType=Capsule, length=28.169, radius=6.5, center=<-6.1534 1.6414 0.2233>, rotation=<rot 0 -88 90>}]
}
constraints [
 {bone1Name=spine_03, bone2Name=spine_01, swing1LimitDegrees=45, swing2LimitDegrees=45, twistLimitDegrees=45}
 {bone1Name=lowerarm_l, bone2Name=spine_05, swing1LimitDegrees=45, swing2LimitDegrees=45, twistLimitDegrees=45}
 {bone1Name=upperarm_twist_01_l, bone2Name=spine_05, swing1LimitDegrees=45, swing2LimitDegrees=45, twistLimitDegrees=45}
 {bone1Name=lowerarm_twist_01_r, bone2Name=spine_05, swing1LimitDegrees=45, swing2LimitDegrees=45, twistLimitDegrees=45}
 {bone1Name=upperarm_twist_01_r, bone2Name=spine_05, swing1LimitDegrees=45, swing2LimitDegrees=45, twistLimitDegrees=45}
 {bone1Name=calf_twist_02_l, bone2Name=pelvis, swing1LimitDegrees=45, swing2LimitDegrees=45, twistLimitDegrees=45}
 {bone1Name=foot_l, bone2Name=pelvis, swing1LimitDegrees=45, swing2LimitDegrees=45, twistLimitDegrees=45}
 {bone1Name=thigh_twist_01_l, bone2Name=pelvis, swing1LimitDegrees=45, swing2LimitDegrees=45, twistLimitDegrees=45}
 {bone1Name=calf_twist_02_r, bone2Name=pelvis, swing1LimitDegrees=45, swing2LimitDegrees=45, twistLimitDegrees=45}
 {bone1Name=foot_r, bone2Name=pelvis, swing1LimitDegrees=45, swing2LimitDegrees=45, twistLimitDegrees=45}
 {bone1Name=thigh_twist_01_r, bone2Name=pelvis, swing1LimitDegrees=45, swing2LimitDegrees=45, twistLimitDegrees=45}
 {
  bone1Name=spine_05
  bone2Name=spine_04
  swing1LimitDegrees=45
  swing1Motion=Free
  swing2LimitDegrees=45
  swing2Motion=Free
  twistLimitDegrees=45
  twistMotion=Free
 }
 {
  bone1Name=spine_04
  bone2Name=spine_03
  swing1LimitDegrees=45
  swing1Motion=Free
  swing2LimitDegrees=45
  swing2Motion=Free
  twistLimitDegrees=45
  twistMotion=Free
 }
 {
  bone1Name=neck_01
  bone2Name=spine_05
  swing1LimitDegrees=45
  swing1Motion=Free
  swing2LimitDegrees=45
  swing2Motion=Free
  twistLimitDegrees=45
  twistMotion=Free
 }
 {
  bone1Name=neck_02
  bone2Name=neck_01
  swing1LimitDegrees=45
  swing1Motion=Free
  swing2LimitDegrees=45
  swing2Motion=Free
  twistLimitDegrees=45
  twistMotion=Free
 }
 {
  bone1Name=head
  bone2Name=neck_02
  swing1LimitDegrees=45
  swing1Motion=Free
  swing2LimitDegrees=45
  swing2Motion=Free
  twistLimitDegrees=45
  twistMotion=Free
 }
 {
  bone1Name=spine_01
  bone2Name=pelvis
  swing1LimitDegrees=45
  swing1Motion=Free
  swing2LimitDegrees=45
  swing2Motion=Free
  twistLimitDegrees=45
  twistMotion=Free
 }
]
